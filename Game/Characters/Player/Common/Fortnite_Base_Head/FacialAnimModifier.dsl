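; FacialAnimModifier is an animation modifier for the Fortnite base player head that cleans facial animation sequences.
; When applied, it removes bone animation from the faceAttach and FACIAL_C_FacialRoot bones.
; The revert event is not implemented, so the removal is one-way within the modifier.
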(event EventOnApply
  (params (AnimationSequence :type Anim%20Sequence%20Object%20Reference))
  (AnimationBlueprintLibrary|Bones|RemoveBoneAnimation AnimationSequence "faceAttach")
  (AnimationBlueprintLibrary|Bones|RemoveBoneAnimation AnimationSequence "FACIAL_C_FacialRoot"))
