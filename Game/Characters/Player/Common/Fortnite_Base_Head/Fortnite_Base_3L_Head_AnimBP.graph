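# Animation Blueprint for the Fortnite base 3L player head, driving facial animation in frontend and gameplay.
# Parent-mesh pose and curves blend with idle faces, blinks and Live Link input by facial mode.
# Animation type selects curve remapping or Rig Mapper, then RigLogic evaluates except in FullHead mode.
# FaceSlot layers on the facial root, and the head bone shrinks to 0.01 scale when the playlist uses custom character parts.
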AnimBlueprint /Game/Characters/Player/Common/Fortnite_Base_Head/Fortnite_Base_3L_Head_AnimBP
target_skeleton=&/Fortnite_Base_3L_Head_Skeleton
graphs [
 AnimGraph {nodes [
   n1 = Get MeshToCopy() -> (MeshToCopy:Skeletal Mesh Component Object Reference)
   Blend Poses by bool(
    BlendPose_0=Layered blend per bone(BasePose=Copy Pose From Mesh(SourceMeshComponent=@n1) {bCopyCurves=true}, BlendPoses_0=Slot 'DefaultSlot'()) {layerSetup [{branchFilters [{boneName=faceAttach}]}]}
    BlendPose_1=Layered blend per bone(BasePose=Copy Pose From Mesh(SourceMeshComponent=@n1) {bCopyCurves=true}, BlendPoses_0=Slot 'inGameSlot'() {slotName=inGameSlot}) {layerSetup [{branchFilters [{boneName=faceAttach}]}]}
    bActiveValue=Get bCanPlayCustomAnimations() -> (bCanPlayCustomAnimations:bool)
   ) {blendType=Linear}
   n2 = Use cached pose 'PrePoseAsset'()
   n3 = Get facial_curve_mapping_expression() -> (facial_curve_mapping_expression:Curve Expressions Data Asset Object Reference)
   Blend Poses by bool(
    BlendPose_1=Blend Poses by bool(
     BlendPose_0=Use cached pose 'PrePoseAsset'()
     BlendPose_1=Remap Curves From Mesh!(
      SourcePose:Pose Link Structure=@n2
      CurveExpressionsDataAsset:Curve Expressions Data Asset Object Reference=@n3
      SourceMeshComponent:Skeletal Mesh Component Weak object
     ) -> (Pose:Pose Link Structure) {is=title AnimGraphNode_RemapCurvesFromMesh, bUseAttachedParent=true, expressionSource=DataAsset}
     bActiveValue=Get bIs3LAnim() -> (bIs3LAnim:bool)
    ) {blendType=Linear}
    bActiveValue=Get bUseRigMapper() -> (bUseRigMapper:bool)
   ) {blendType=Linear}
   Blend Poses by bool(
    BlendPose_0=Remap Curves From Mesh(
     SourcePose=Use cached pose 'DefaultFaceIdle'()
     CurveExpressionsDataAsset=Get Facial_Curve_Zero_Expression() -> (Facial_Curve_Zero_Expression:Curve Expressions Data Asset Object Reference)
    ) {bUseAttachedParent=true, expressionSource=DataAsset}
    bActiveValue=Get DefaultFaceIdleActive() -> (DefaultFaceIdleActive:bool)
   ) {blendType=Linear}
   n4 = Make Expression Map!() -> (Map:Map of Names to Strings) {
    is=title K2Node_MakeCurveExpressionMap
    expressions {assignmentExpressions="CTRL_expressions_tongueUp=tongue_up_pose*0.65+tongue_forward_pose*0.1\r\nCTRL_expressions_mouthUpperLipBiteL=phoneme_mbp_pose*0.5+phoneme_fv_pose*0.1\r\nCTRL_expressions_mouthUpperLipBiteR=phoneme_mbp_pose*0.5+phoneme_fv_pose*0.1\r\nCTRL_expressions_mouthLowerLipBiteL=clamp((phoneme_mbp_pose*0.35+phoneme_fv_pose*0.727+L_lower_lip_up_pose*0.5), 0, 1)\r\nCTRL_expressions_mouthLowerLipBiteR=clamp((phoneme_mbp_pose*0.35+phoneme_fv_pose*0.727+R_lower_lip_up_pose*0.5), 0, 1)\r\nCTRL_expressions_mouthUpperLipTowardsTeethL=clamp((L_upper_lip_lower_pose*0.6+L_smile_pose*0.5), 0, 1)\r\nCTRL_expressions_mouthCornerDownL=L_upper_lip_lower_pose*0.4\r\nCTRL_expressions_mouthUpperLipTowardsTeethR=clamp((R_upper_lip_lower_pose*0.6+R_smile_pose*0.5), 0, 1)\r\nCTRL_expressions_mouthCornerDownR=R_upper_lip_lower_pose*0.4\r\nCTRL_expressions_eyeLookRightL=clamp((look_right_pose*0.6+eye_convergence_pose*0.528+look_down_pose*0.087), 0, 1)\r\nCTRL_expressions_eyeLookRightR=look_right_pose*0.6\r\nCTRL_expressions_jawChinRaiseDL=clamp((jaw_up_pose*0.5+phoneme_fv_pose*0.285+R_lower_lip_down_pose*0.19+jaw_left_pose*0.2+mouth_up_pose*0.272), 0, 1)\r\nCTRL_expressions_jawChinRaiseDR=clamp((jaw_up_pose*0.5+phoneme_fv_pose*0.285+L_lower_lip_down_pose*0.19+mouth_up_pose*0.272), 0, 1)\r\nCTRL_expressions_jawChinRaiseUL=jaw_up_pose*0.1+phoneme_fv_pose*0.3+phoneme_ch_pose*0.122+mouth_up_pose*0.4\r\nCTRL_expressions_jawChinRaiseUR=jaw_up_pose*0.1+phoneme_fv_pose*0.3+phoneme_ch_pose*0.122+mouth_up_pose*0.4\r\nCTRL_expressions_browDownL=clamp((L_squeeze_pose*0.552+C_nose_wrinkler_pose*0.119+L_brow_down_pose*0.653+C_glabella_up_pose*0.167), 0, 1)\r\nCTRL_expressions_browUINL=L_squeeze_pose*0.269+L_brow_down_pose*0.3\r\nCTRL_expressions_eyeCheekRaiseL=clamp((L_squeeze_pose+C_nose_wrinkler_pose+L_cheek_up_pose), 0, 1)\r\nCTRL_expressions_eyeLowerLidDownL=L_squeeze_pose*0.1+C_nose_wrinkler_pose*0.3+L_cheek_up_pose*0.25\r\nCTRL_expressions_eyelidRollNegDL=L_squeeze_pose*0.1+C_nose_wrinkler_pose*0.3+L_cheek_up_pose*0.2\r\nCTRL_expressions_noseWrinkleL=L_squeeze_pose*0.452+C_nose_wrinkler_pose*0.345\r\nCTRL_expressions_noseWrinkleUpperL=clamp((L_squeeze_pose+C_nose_wrinkler_pose), 0, 1)\r\nCTRL_expressions_noseSquash=L_squeeze_pose*0.095+R_squeeze_pose*0.095\r\nCTRL_expressions_browDownR=clamp((R_squeeze_pose*0.552+C_nose_wrinkler_pose*0.119+R_brow_down_pose*0.653+C_glabella_up_pose*0.165), 0, 1)\r\nCTRL_expressions_browUINR=R_squeeze_pose*0.269+R_brow_down_pose*0.3\r\nCTRL_expressions_eyeCheekRaiseR=clamp((R_squeeze_pose+C_nose_wrinkler_pose+R_cheek_up_pose), 0, 1)\r\nCTRL_expressions_eyeLowerLidDownR=R_squeeze_pose*0.1+C_nose_wrinkler_pose*0.3+R_cheek_up_pose*0.25\r\nCTRL_expressions_eyelidRollNegDR=clamp((R_squeeze_pose*0.1+look_down_pose+R_cheek_up_pose*0.2), 0, 1)\r\nCTRL_expressions_noseWrinkleR=R_squeeze_pose*0.452+C_nose_wrinkler_pose*0.345\r\nCTRL_expressions_noseWrinkleUpperR=clamp((R_squeeze_pose+C_nose_wrinkler_pose), 0, 1)\r\nCTRL_expressions_eyeLookLeftL=look_left_pose*0.6\r\nCTRL_expressions_eyeLookLeftR=clamp((look_left_pose*0.6+eye_convergence_pose*0.398+look_down_pose*0.167), 0, 1)\r\nCTRL_expressions_tongueIn=clamp((tongue_back_pose+tongue_down_pose*0.7), 0, 1)\r\nCTRL_expressions_tongueRollDown=clamp((tongue_back_pose*0.5+tongue_down_pose+tongue_forward_pose*0.2), 0, 1)\r\nCTRL_expressions_eyeLookUpL=look_up_pose\r\nCTRL_expressions_eyeLookUpR=look_up_pose\r\nCTRL_expressions_mouthUpperLipRaiseL=clamp((phoneme_fv_pose*0.2+phoneme_oo_pose*0.1+jaw_open_pose*0.25+L_upper_lip_raiser_pose*0.5+R_upper_lip_raiser_pose*0.2+phoneme_ch_pose*0.15+L_frown_pose*0.169+L_smile_pose*0.25+R_smile_pose*0.1), 0, 1)\r\nCTRL_expressions_mouthUpperLipRaiseR=clamp((phoneme_fv_pose*0.2+phoneme_oo_pose*0.1+jaw_open_pose*0.25+L_upper_lip_raiser_pose*0.2+R_upper_lip_raiser_pose*0.5+phoneme_ch_pose*0.15+R_frown_pose*0.169+L_smile_pose*0.1+R_smile_pose*0.25), 0, 1)\r\nCTRL_expressions_jawBack=clamp((jaw_back_pose+phoneme_oo_pose*0.6+jaw_open_pose*0.161), 0, 1)\r\nCTRL_expressions_mouthLowerLipDepressL=clamp((L_lower_lip_down_pose*0.65+phoneme_oo_pose*0.109+jaw_open_pose*0.313+L_lower_lip_up_pose*0.25+phoneme_ch_pose*0.25+L_frown_pose*0.1+L_smile_pose*0.1), 0, 1)\r\nCTRL_expressions_mouthFunnelDL=clamp((L_lower_lip_down_pose*0.2+phoneme_oo_pose+phoneme_ch_pose*0.577), 0, 1)\r\nCTRL_expressions_mouthCornerRounderDL=L_lower_lip_down_pose*0.4\r\nCTRL_expressions_mouthLowerLipDepressR=clamp((R_lower_lip_down_pose*0.65+phoneme_oo_pose*0.109+jaw_open_pose*0.313+R_lower_lip_up_pose*0.25+phoneme_ch_pose*0.25+R_frown_pose*0.1+R_smile_pose*0.1), 0, 1)\r\nCTRL_expressions_mouthFunnelDR=clamp((R_lower_lip_down_pose*0.2+phoneme_oo_pose+phoneme_ch_pose*0.584), 0, 1)\r\nCTRL_expressions_mouthCornerRounderDR=R_lower_lip_down_pose*0.4\r\nCTRL_expressions_eyeBlinkL=L_blink_pose\r\nCTRL_expressions_eyeBlinkR=R_blink_pose\r\nCTRL_expressions_mouthCornerNarrowL=L_lip_corner_narrow_pose\r\nCTRL_expressions_mouthCornerNarrowR=R_lip_corner_narrow_pose\r\nCTRL_expressions_eyeWidenL=clamp((look_down_pose+L_wide_pose), 0, 1)\r\nCTRL_expressions_eyeWidenR=clamp((look_down_pose+R_wide_pose), 0, 1)\r\nCTRL_expressions_eyeUpperLidUpL=look_down_pose*0.1+C_nose_wrinkler_pose*0.1+L_cheek_up_pose*0.15\r\nCTRL_expressions_eyeUpperLidUpR=look_down_pose*0.1+C_nose_wrinkler_pose*0.1+R_cheek_up_pose*0.15\r\nCTRL_expressions_eyeLowerLidUpL=look_down_pose*0.7\r\nCTRL_expressions_eyeLowerLidUpR=look_down_pose*0.7\r\nCTRL_expressions_eyelidRollPosUL=look_down_pose*0.3\r\nCTRL_expressions_eyelidRollPosDL=clamp((look_down_pose+L_squint_inner_pose*0.2), 0, 1)\r\nCTRL_expressions_eyelidRollNegUR=look_down_pose*0.3\r\nCTRL_expressions_eyeLookDownL=look_down_pose\r\nCTRL_expressions_eyeLookDownR=look_down_pose\r\nCTRL_expressions_mouthUp=phoneme_oo_pose*0.005+mouth_up_pose*0.795\r\nCTRL_expressions_mouthLipsPurseUL=phoneme_oo_pose*0.5+phoneme_ch_pose*0.4\r\nCTRL_expressions_mouthLipsPurseUR=phoneme_oo_pose*0.5+phoneme_ch_pose*0.4\r\nCTRL_expressions_mouthLipsPurseDL=phoneme_oo_pose*0.5+phoneme_ch_pose*0.4\r\nCTRL_expressions_mouthLipsPurseDR=phoneme_oo_pose*0.5+phoneme_ch_pose*0.4\r\nCTRL_expressions_mouthLipsTowardsUL=phoneme_oo_pose*0.3+phoneme_ch_pose*0.3\r\nCTRL_expressions_mouthLipsTowardsUR=phoneme_oo_pose*0.3+phoneme_ch_pose*0.3\r\nCTRL_expressions_mouthLipsTowardsDL=phoneme_oo_pose*0.3+phoneme_ch_pose*0.3\r\nCTRL_expressions_mouthLipsTowardsDR=phoneme_oo_pose*0.3+phoneme_ch_pose*0.3\r\nCTRL_expressions_mouthFunnelUL=clamp((phoneme_oo_pose+phoneme_ch_pose*0.577), 0, 1)\r\nCTRL_expressions_mouthFunnelUR=clamp((phoneme_oo_pose+phoneme_ch_pose*0.584), 0, 1)\r\nCTRL_expressions_mouthCornerWideL=clamp((phoneme_oo_pose+L_lip_corner_wide_pose), 0, 1)\r\nCTRL_expressions_mouthCornerWideR=clamp((phoneme_oo_pose+R_lip_corner_wide_pose), 0, 1)\r\nCTRL_expressions_mouthPressDL=jaw_left_pose*0.212\r\nCTRL_expressions_jawOpen=clamp((jaw_left_pose*0.056+jaw_open_pose), 0, 1)\r\nCTRL_expressions_jawLeft=jaw_left_pose*0.805\r\nCTRL_expressions_mouthDimpleL=clamp((L_lip_corner_wide_pose*0.47+jaw_open_pose*0.35+L_smile_pose*0.3), 0, 1)\r\nCTRL_expressions_mouthDimpleR=clamp((R_lip_corner_wide_pose*0.47+jaw_open_pose*0.35+R_smile_pose*0.3), 0, 1)\r\nCTRL_expressions_eyelidRollPosDR=C_nose_wrinkler_pose*0.3+R_squint_inner_pose*0.2\r\nCTRL_expressions_noseStretch=jaw_open_pose*0.143\r\nCTRL_expressions_mouthCornerPullL=jaw_open_pose*0.3+L_smile_pose*0.6\r\nCTRL_expressions_mouthCornerPullR=jaw_open_pose*0.3+R_smile_pose*0.6\r\nCTRL_expressions_mouthStretchL=jaw_open_pose*0.2+L_frown_pose*0.1+L_smile_pose*0.1\r\nCTRL_expressions_mouthStretchR=jaw_open_pose*0.2+R_frown_pose*0.1+R_smile_pose*0.1\r\nCTRL_expressions_mouthCornerDepressL=jaw_open_pose*0.2+L_frown_pose*0.7\r\nCTRL_expressions_mouthCornerDepressR=jaw_open_pose*0.2+R_frown_pose*0.7\r\nCTRL_expressions_mouthCornerRounderUL=clamp((jaw_open_pose*0.2+L_upper_lip_raiser_pose), 0, 1)\r\nCTRL_expressions_mouthCornerRounderUR=clamp((jaw_open_pose*0.2+R_upper_lip_raiser_pose), 0, 1)\r\nCTRL_expressions_jawOpenExtreme=jaw_open_pose*0.4\r\nCTRL_expressions_browRaiseInL=clamp((L_brow_up_pose*0.4+C_glabella_up_pose*0.8), 0, 1)\r\nCTRL_expressions_browRaiseOuterL=L_brow_up_pose*0.7+C_glabella_up_pose*0.25\r\nCTRL_expressions_browRaiseInR=clamp((R_brow_up_pose*0.4+C_glabella_up_pose*0.8), 0, 1)\r\nCTRL_expressions_browRaiseOuterR=R_brow_up_pose*0.7+C_glabella_up_pose*0.25\r\nCTRL_expressions_mouthLowerLipTowardsTeethL=clamp((L_lower_lip_up_pose+L_frown_pose*0.344), 0, 1)\r\nCTRL_expressions_mouthCornerUpL=L_lower_lip_up_pose*0.1+L_smile_pose*0.3+mouth_up_pose*0.3\r\nCTRL_expressions_mouthLowerLipTowardsTeethR=clamp((R_lower_lip_up_pose+R_frown_pose*0.344), 0, 1)\r\nCTRL_expressions_mouthCornerUpR=R_lower_lip_up_pose*0.1+R_smile_pose*0.3+mouth_up_pose*0.3\r\nCTRL_expressions_jawFwd=jaw_forward_pose\r\nCTRL_expressions_eyeSquintInnerL=L_squint_inner_pose\r\nCTRL_expressions_eyeSquintInnerR=R_squint_inner_pose\r\nCTRL_expressions_tongueDown=tongue_down_pose\r\nCTRL_expressions_browLateralL=L_brow_squeeze_in_pose\r\nCTRL_expressions_browLateralR=R_brow_squeeze_in_pose\r\nCTRL_expressions_jawRight=jaw_right_pose\r\nCTRL_expressions_mouthDown=mouth_down_pose\r\nCTRL_expressions_browDINL=C_glabella_up_pose*0.284\r\nCTRL_expressions_browDINR=C_glabella_up_pose*0.284\r\nCTRL_expressions_mouthCornerSharpenUL=phoneme_ch_pose*0.1\r\nCTRL_expressions_mouthCornerSharpenUR=phoneme_ch_pose*0.1\r\nCTRL_expressions_mouthCornerSharpenDL=phoneme_ch_pose*0.1\r\nCTRL_expressions_mouthCornerSharpenDR=phoneme_ch_pose*0.1\r\nCTRL_expressions_tongueOut=tongue_forward_pose*0.9\r\nCTRL_expressions_tongueWide=tongue_forward_pose\r\nCTRL_expressions_eyeRelaxL=L_upperLidDown_pose\r\nCTRL_expressions_eyeRelaxR=R_upperLidDown_pose"}
   }
   Remap Curves From Mesh!(SourcePose, CurveExpressions:Map of Names to Strings=@n4, SourceMeshComponent) -> (Pose) {bUseAttachedParent=true, expressionSource=ExpressionMap}
   n5 = "Transform (Modify) Bone - Bone: head"!(ComponentPose=Local To Component(LocalPose=Use cached pose 'preScaleCheck'()), Alpha=1) -> (Pose) {scaleMode=BMM_Replace, alphaBoolBlend {blendOption=Linear}, boneToModify {boneName=head}, scale=<0.01 0.01 0.01>}
   Output Pose(Result=Blend Poses by bool(
     BlendPose_0=Component To Local(ComponentPose=@n5)
     BlendPose_1=Use cached pose 'preScaleCheck'()
     BlendTime_0=0
     BlendTime_1=0
     bActiveValue=Get GetPlayerListUsesCustomCharacterParts() -> (GetPlayerListUsesCustomCharacterParts:bool)
    ) {blendType=Linear}) {layerGroup=None}
   n6 = "Blend Poses (EFortFacialAnimTypes)"!(
    BlendPose_0:Pose Link Structure=Use cached pose 'GameFacial'()
    BlendPose_1:Pose Link Structure=Layered blend per bone(BasePose=Use cached pose 'GameFacial'(), BlendPoses_0=Use cached pose 'LiveLinkFacial'()) {layerSetup [{branchFilters [{boneName=FACIAL_C_FacialRoot}]}]}
    BlendPose_2:Pose Link Structure=Use cached pose 'LiveLinkFacial'()
    BlendTime_0:float=0.1
    BlendTime_1:float=0.1
    BlendTime_2:float=0.1
    ActiveEnumValue:Byte=Get!() -> (
     FacialInput_SubjectName:Live Link Subject Name Structure
     FacialInput_CurrentAnimType:EFortFacialAnimTypes Enum
     FacialInput_bCurvesOnly:bool=~
    ).FacialInput_CurrentAnimType
   ) -> (Pose:Pose Link Structure) {
    is=title AnimGraphNode_BlendListByEnum
    blendType=Linear
    blendPose [{linkId=-1, sourceLinkId=-1}, {linkId=-1, sourceLinkId=-1}, {linkId=-1, sourceLinkId=-1}]
    blendTime [0.1,0.1,0.1]
   }
   n7 = Use cached pose 'IncomingPose'()
   n8 = Get bBypassHeadPartLogic() -> (bBypassHeadPartLogic:bool)
   Save cached pose 'PrePoseAsset'(Pose=Blend Poses by bool(
     BlendPose_0=Blend Poses by bool(BlendPose_0=@n7, BlendPose_1=@n6, BlendTime_0=0, BlendTime_1=0, bActiveValue=@n8)
     BlendPose_1=Use cached pose 'IdlePose'()
     bActiveValue=Get bEnableGameplayFacialCurves() -> (bEnableGameplayFacialCurves:bool)
    )) {cacheName=PrePoseAsset}
   n9 = Two Way Blend!(A=Use cached pose 'IdlePose'(), B=Use cached pose 'IncomingPoseWithBlinks'(), AlphaCurveName:Name=DisableFaceOverride) -> (Pose) {blendNode {alphaInputType=Curve, alphaBoolBlend {blendOption=Linear}}}
   n10 = Use cached pose 'IncomingPose'()
   n11 = Get bypassIdleAnim() -> (bypassIdleAnim:bool)
   n12 = "Blend Poses (EFortFacialAnimTypes)"(
    BlendPose_0=Two Way Blend!(
     A=Use cached pose 'IncomingPoseWithBlinks'()
     B=Blend Poses by bool(BlendPose_0=@n10, BlendPose_1=@n9, bActiveValue=@n11) {blendType=Linear}
     Alpha:float=Get FaceIdleAlpha() -> (FaceIdleAlpha:float)
    ) -> (Pose) {blendNode {alphaBoolBlend {blendOption=Linear}}}
    BlendPose_1=Use cached pose 'IncomingPose'()
    BlendPose_2=Use cached pose 'IncomingPose'()
    BlendTime_0=0.1
    BlendTime_1=0.1
    BlendTime_2=0.1
    ActiveEnumValue=Get() -> (FacialInput_bCurvesOnly=~).FacialInput_CurrentAnimType
   ) {
    blendType=Linear
    blendPose [{linkId=-1, sourceLinkId=-1}, {linkId=-1, sourceLinkId=-1}, {linkId=-1, sourceLinkId=-1}]
    blendTime [0.1,0.1,0.1]
   }
   Save cached pose 'GameFacial'(Pose=@n12) {cacheName=GameFacial}
   Save cached pose 'IncomingPoseWithBlinks'(Pose=Blend Facial Curves!(
     FirstPose:Pose Link Structure=Use cached pose 'IncomingPose'()
     SecondPose:Pose Link Structure=Use cached pose 'DefaultFaceIdle'()
     Alpha:float=Get FaceIdleBlinksAlpha() -> (FaceIdleBlinksAlpha:float)
    ) -> (Pose:Pose Link Structure) {is=title FortAnimGraphNode_BlendFacialCurves, bIs3LRig=true}) {cacheName=IncomingPoseWithBlinks}
   n13 = Slot 'FaceSlot'(Source=Use cached pose 'PostPoseAsset'()) {slotName=FaceSlot}
   n14 = Use cached pose 'PostPoseAsset'()
   n15 = Get FaceSlotWeight() -> (FaceSlotWeight:float)
   Save cached pose 'preScaleCheck'(Pose=Layered blend per bone(BasePose=Use cached pose 'PostPoseAsset'(), BlendPoses_0=Two Way Blend(A=@n14, B=@n13, Alpha=@n15)) {layerSetup [{branchFilters [{boneName=FACIAL_C_FacialRoot}]}]}) {cacheName=preScaleCheck}
   n16 = Use cached pose 'GameFacial'()
   n17 = Break Fort Anim Input Facial!(FortAnimInput_Facial:"Fort Anim Input Facial Structure (by ref)"=Get FacialInput() -> (FacialInput:Fort Anim Input Facial Structure)) -> (SubjectName:Live Link Subject Name Structure, CurrentAnimType:EFortFacialAnimTypes Enum, bCurvesOnly:bool=~) {is=title K2Node_BreakStruct}
   n18 = Live Link Pose!(
    InputPose:Pose Link Structure=Use cached pose 'GameFacial'()
    LiveLinkSubjectName:Live Link Subject Name Structure=@n17
    bDoLiveLinkEvaluation:bool=True
   ) -> (Pose:Pose Link Structure) {
    is=title AnimGraphNode_LiveLinkPose
    retargetAsset=/Game/Characters/Player/Common/RemapAssets/FacialAnim_CurvesAndBones_HeadRemapAsset.FacialAnim_CurvesAndBones_HeadRemapAsset_C
   }
   Save cached pose 'LiveLinkFacial'(Pose=Blend Poses by bool(
     BlendPose_0=Live Link Pose(InputPose=@n16, LiveLinkSubjectName=@n17, bDoLiveLinkEvaluation=True) {retargetAsset=/Game/Characters/Player/Common/RemapAssets/FacialAnim_HeadRemapAsset.FacialAnim_HeadRemapAsset_C}
     BlendPose_1=@n18
     bActiveValue=@n17.bCurvesOnly
    ) {blendType=Linear}) {cacheName=LiveLinkFacial}
   n19 = "Blend Poses (EFortniteAnimationType)"!(
    BlendPose_0:Pose Link Structure
    BlendPose_1:Pose Link Structure=Remap Curves!(
     SourcePose:Pose Link Structure=Use cached pose 'PrePoseAsset'()
     CurveExpressionsDataAsset:Curve Expressions Data Asset Object Reference=Get facial_curve_mapping_expression()
    ) -> (Pose:Pose Link Structure) {is=title AnimGraphNode_RemapCurves, expressionSource=DataAsset}
    BlendPose_2:Pose Link Structure=Use cached pose 'PrePoseAsset'()
    BlendPose_3:Pose Link Structure=Rig Mapper!(SourcePose:Pose Link Structure=Use cached pose 'PrePoseAsset'(), Alpha:float=1) -> (Pose:Pose Link Structure) {is=title AnimGraphNode_RigMapper, definitions [/RigMapper/Definitions/Baked/RM_MHL_CDL]}
    BlendTime_0:float=0.1
    BlendTime_1:float=0.1
    BlendTime_2:float=0.1
    BlendTime_3:float=0.1
    ActiveEnumValue:Byte=Get AnimationType() -> (AnimationType:EFortniteAnimationType Enum=FN_Legacy)
   ) -> (Pose:Pose Link Structure) {
    is=title AnimGraphNode_BlendListByEnum
    blendType=Linear
    blendPose [{linkId=-1, sourceLinkId=-1}, {linkId=-1, sourceLinkId=-1}, {linkId=-1, sourceLinkId=-1}, {linkId=-1, sourceLinkId=-1}]
    blendTime [0.1,0.1,0.1,0.1]
   }
   Save cached pose 'PreRigLogic'(Pose=Remap Curves(
     SourcePose=@n19
     CurveExpressionsDataAsset=Get facial_curve_mute_expression() -> (facial_curve_mute_expression:Curve Expressions Data Asset Object Reference)
    ) {expressionSource=DataAsset}) {cacheName=PreRigLogic}
   Save cached pose 'IncomingPose'(Pose=Copy Pose From Mesh() {bCopyCurves=true, bUseAttachedParent=true}) {cacheName=IncomingPose}
   Save cached pose 'DefaultFaceIdle'(Pose=Sequence Player 'Gameplay_Default_3L_Face_Idle'() {tag=DefaultIdle, sequence=/Game/Animation/Game/MainPlayer/Menu/FACIAL/Gameplay_Default_3L_Face_Idle}) {cacheName=DefaultFaceIdle}
   n20 = Use cached pose 'PreRigLogic'()
   n21 = Get() -> (FacialInput_bCurvesOnly=~)
   Save cached pose 'PostPoseAsset'(Pose=Blend Poses by bool(
     BlendPose_0=Use cached pose 'PreRigLogic'()
     BlendPose_1=AnimNode_RigLogic!(AnimSequence:Pose Link Structure=@n20) -> (Pose:Pose Link Structure) {is=title AnimGraphNode_RigLogic}
     bActiveValue="Equal (Enum)"(A:EFortFacialAnimTypes Enum=@n21.FacialInput_CurrentAnimType, B:EFortFacialAnimTypes Enum=FullHead)
    ) {blendType=Linear}) {cacheName=PostPoseAsset}
   n22 = Two Way Blend!(
    A=Blend Poses by bool(
     BlendPose_0=Sequence Player 'Frontend_Default_3L_Face_Idle'() {sequence=/Game/Animation/Game/MainPlayer/Menu/FACIAL/Frontend_Default_3L_Face_Idle}
     BlendPose_1=Use cached pose 'DefaultFaceIdle'()
     bActiveValue=Get SkipDefaultFaceIdle() -> (SkipDefaultFaceIdle:bool)
    ) {blendType=Linear}
    B=Slot 'Frontend_CustomFaceIdle'() {slotName=Frontend_CustomFaceIdle}
    AlphaCurveName:Name=Custom_Frontend_Face_Idle
   ) -> (Pose) {blendNode {alphaInputType=Curve, alphaBoolBlend {blendOption=Linear}}}
   n23 = Use cached pose 'DefaultFaceIdle'()
   n24 = Get bIsInFrontEnd()
   Save cached pose 'IdlePose'(Pose=Layered blend per bone(
     BasePose=Use cached pose 'IncomingPose'()
     BlendPoses_0=Blend Poses by bool(BlendPose_0=@n22, BlendPose_1=@n23, bActiveValue=@n24) {blendType=Linear}
    ) {layerSetup [{branchFilters [{boneName=FACIAL_C_FacialRoot}]}]}) {cacheName=IdlePose}
   Sequence Player 'Fortnite_Base_Head_ROM'() {sequence=/Game/Characters/Player/Common/Fortnite_Base_Head/Anim_Tests/Fortnite_Base_Head_ROM}
   GetPlaylistUsesCustomCharacterParts!() -> (ReturnValue:bool)
  ]}
 EventGraph {nodes [
   Event BlueprintInitializeAnimation()
   > Set MeshToCopy(MeshToCopy:Skeletal Mesh Component Object Reference=GetOwnerSkeletalMeshComponent()) -> (Output_Get:Skeletal Mesh Component Object Reference)
   > Set GetPlayerListUsesCustomCharacterParts(GetPlayerListUsesCustomCharacterParts:bool=GetPlaylistUsesCustomCharacterParts()) -> (Output_Get:bool)
   n1 = Get Class(Object=GetOwningActor())
   > n2 = Branch(Condition="Equal Exactly (String)"(A=Get Class Display Name(Class=@n1), B=AnimationEditorPreviewActor))
   Set bypassIdleAnim(execute=[@n2.then, @n3.Is Not Valid], bypassIdleAnim:bool=true) -> (Output_Get:bool)
   n3 = Is Valid(exec=@n2.else, InputObject=TryGetPawnOwner())
   > Set bypassIdleAnim(bypassIdleAnim=false)
   Branch(Condition=Get bIsInFrontEnd())
   > n4 = Set DefaultFaceIdleActive(DefaultFaceIdleActive:bool="float > float"(A=GetCurveValue(CurveName=BaseHeadDefaultFaceIdle))) -> (Output_Get:bool)
   > Branch(Condition=@n4)
   > Set SkipDefaultFaceIdle(SkipDefaultFaceIdle:bool=true) -> (Output_Get:bool)
   n5 = GetPartSkeletalMeshComponent()
   n6 = Is Valid(InputObject=@n5)
   n7 = GetAnimInstance(self=@n5)
   > n8 = Is Valid(InputObject=@n7)
   > n9 = Branch(Condition="float > float"(A=GetCurveValue(self=@n7, CurveName=DisableFaceOverride), B=0.1))
   > Set is_3L_anim_obsolete(is_3L_anim_obsolete:bool="float > float"(A=GetCurveValue(self=@n7, CurveName=is_3L), B=0.1)) -> (Output_Get:bool)
   n10 = Set is_3L_anim_obsolete(execute=[@n9.else, @n6.Is Not Valid, @n8.Is Not Valid], is_3L_anim_obsolete=true)
   PrintString!(execute, InString="To String (Boolean)"(InBool=@n10), bPrintToScreen, bPrintToLog, TextColor, Duration, Key) -> (then)
   TryGetPawnOwner()
   GetOwnerSkeletalMeshComponent()
  ]}
]
variables [
 bypassIdleAnim {type=boolean}
 DefaultFaceIdleActive {type=boolean}
 facial_curve_mapping_expression {type=CurveExpressionsDataAsset, default=/Game/Characters/Player/Common/Fortnite_Base_Head/Facials/CurveMappings/FN_LegacyTo3L_Main_Mapping}
 facial_curve_mute_expression {type=CurveExpressionsDataAsset, default=null}
 Facial_Curve_Zero_Expression {type=CurveExpressionsDataAsset, default=/Game/Characters/Player/Common/Fortnite_Base_Head/Facials/CurveMappings/FN_ZeroOut3LDefaultFaceIdle}
 GetPlayerListUsesCustomCharacterParts {type=boolean}
 is_3L_anim_obsolete {type=boolean, default=true}
 MeshToCopy {type=SkeletalMeshComponent}
 SkipDefaultFaceIdle {type=boolean}
]
