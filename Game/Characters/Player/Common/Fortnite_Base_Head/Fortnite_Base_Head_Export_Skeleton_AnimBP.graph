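# Animation Blueprint for the Fortnite player base head skeleton, driving facial animation on top of the pose copied from the attached parent mesh.
# Blends idle, blinks, gameplay and Live Link faces, then converts curves through expression remaps or Rig Mapper.
# An optional flipbook Control Rig feeds the facial Pose Asset, followed by FaceSlot and a playlist-dependent head-bone shrink.
# The EventGraph flags flipbook, look and blink curve changes to material parameters.
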AnimBlueprint /Game/Characters/Player/Common/Fortnite_Base_Head/Fortnite_Base_Head_Export_Skeleton_AnimBP
target_skeleton=&/Fortnite_Base_Head_Export_Skeleton
graphs [
 AnimGraph {nodes [
   n1 = Get MeshToCopy() -> (MeshToCopy:Skeletal Mesh Component Object Reference)
   Blend Poses by bool(
    BlendPose_0=Layered blend per bone(BasePose=Copy Pose From Mesh(SourceMeshComponent=@n1) {bCopyCurves=true}, BlendPoses_0=Slot 'DefaultSlot'()) {layerSetup [{branchFilters [{boneName=faceAttach}]}]}
    BlendPose_1=Layered blend per bone(BasePose=Copy Pose From Mesh(SourceMeshComponent=@n1) {bCopyCurves=true}, BlendPoses_0=Slot 'inGameSlot'() {slotName=inGameSlot}) {layerSetup [{branchFilters [{boneName=faceAttach}]}]}
    bActiveValue=Get bCanPlayCustomAnimations() -> (bCanPlayCustomAnimations:bool)
   ) {blendType=Linear}
   Blend Poses by bool(
    BlendPose_0=Remap Curves From Mesh!(
     SourcePose:Pose Link Structure=Use cached pose 'DefaultFaceIdle'()
     CurveExpressionsDataAsset:Curve Expressions Data Asset Object Reference=Get Facial_Curve_Zero_Expression() -> (Facial_Curve_Zero_Expression:Curve Expressions Data Asset Object Reference)
     SourceMeshComponent:Skeletal Mesh Component Weak object
    ) -> (Pose:Pose Link Structure) {is=title AnimGraphNode_RemapCurvesFromMesh, bUseAttachedParent=true, expressionSource=DataAsset}
    bActiveValue=Get bDefaultFaceIdleActive() -> (bDefaultFaceIdleActive:bool)
   ) {blendType=Linear}
   n2 = Make Expression Map!() -> (Map:Map of Names to Strings) {
    is=title K2Node_MakeCurveExpressionMap
    expressions {assignmentExpressions="L_brow_down_pose=clamp((CTRL_expressions_browDownL+CTRL_expressions_browDOUTL*0.648), 0, 1)\r\nR_brow_down_pose=clamp((CTRL_expressions_browDownR+CTRL_expressions_browDOUTR*0.648), 0, 1)\r\nL_brow_squeeze_in_pose=CTRL_expressions_browLateralL\r\nR_brow_squeeze_in_pose=CTRL_expressions_browLateralR\r\nC_glabella_up_pose=clamp((CTRL_expressions_browRaiseInL+CTRL_expressions_browRaiseInR+CTRL_expressions_browUINL+CTRL_expressions_browUINR), 0, 1)\r\nL_brow_up_pose=clamp((CTRL_expressions_browRaiseOuterL*0.896+CTRL_expressions_browUOUTL+CTRL_expressions_eyeCheekRaiseL*0.104), 0, 1)\r\nR_brow_up_pose=clamp((CTRL_expressions_browRaiseOuterR*0.896+CTRL_expressions_browUOUTR+CTRL_expressions_eyeCheekRaiseR*0.104), 0, 1)\r\nL_blink_pose=clamp((CTRL_expressions_eyeBlinkL+CTRL_expressions_eyeRelaxL*0.5+CTRL_expressions_eyeCheekRaiseL*0.104), 0, 1)\r\nR_blink_pose=clamp((CTRL_expressions_eyeBlinkR+CTRL_expressions_eyeRelaxR*0.5+CTRL_expressions_eyeCheekRaiseR*0.104), 0, 1)\r\nL_wide_pose=clamp((CTRL_expressions_eyeWidenL+CTRL_expressions_eyeUpperLidUpL*0.648), 0, 1)\r\nR_wide_pose=clamp((CTRL_expressions_eyeWidenR+CTRL_expressions_eyeUpperLidUpR*0.648), 0, 1)\r\nL_squint_inner_pose=clamp((CTRL_expressions_eyeSquintInnerL+CTRL_expressions_eyeCheekRaiseL*0.5), 0, 1)\r\nR_squint_inner_pose=clamp((CTRL_expressions_eyeSquintInnerR+CTRL_expressions_eyeCheekRaiseR*0.5), 0, 1)\r\nlook_up_pose=clamp((CTRL_expressions_eyeLookUpL+CTRL_expressions_eyeLookUpR), 0, 1)\r\nlook_down_pose=clamp((CTRL_expressions_eyeLookDownL+CTRL_expressions_eyeLookDownR), 0, 1)\r\nlook_left_pose=clamp((CTRL_expressions_eyeLookLeftL+CTRL_expressions_eyeLookLeftR), 0, 1)\r\nlook_right_pose=clamp((CTRL_expressions_eyeLookRightL+CTRL_expressions_eyeLookRightR), 0, 1)\r\nL_squeeze_pose=clamp((CTRL_expressions_noseWrinkleL+CTRL_expressions_eyeCheekRaiseL*0.352), 0, 1)\r\nR_squeeze_pose=clamp((CTRL_expressions_noseWrinkleR+CTRL_expressions_eyeCheekRaiseR*0.352), 0, 1)\r\nmouth_up_pose=CTRL_expressions_mouthUp*0.784\r\nmouth_down_pose=CTRL_expressions_mouthDown*0.648\r\nL_upper_lip_raiser_pose=clamp((CTRL_expressions_mouthUpperLipRaiseL+CTRL_expressions_mouthPressUL*0.5+CTRL_expressions_mouthFunnelUL*0.125+CTRL_expressions_mouthFunnelUR*0.125+CTRL_expressions_mouthFunnelDL*0.125+CTRL_expressions_mouthFunnelDR*0.125+CTRL_expressions_mouthCornerRounderUL*0.5+CTRL_expressions_jawChinRaiseUL*0.812+CTRL_expressions_jawChinRaiseUR*0.028+CTRL_expressions_mouthLeft*0.216+CTRL_expressions_mouthLipsTowardsUL*0.216+CTRL_expressions_mouthLipsTowardsDL*0.216), 0, 1)\r\nR_upper_lip_raiser_pose=clamp((CTRL_expressions_mouthUpperLipRaiseR+CTRL_expressions_mouthPressUR*0.5+CTRL_expressions_mouthFunnelUL*0.125+CTRL_expressions_mouthFunnelUR*0.125+CTRL_expressions_mouthFunnelDL*0.125+CTRL_expressions_mouthFunnelDR*0.125+CTRL_expressions_mouthCornerRounderUR*0.5+CTRL_expressions_jawChinRaiseUL*0.028+CTRL_expressions_jawChinRaiseUR*0.812+CTRL_expressions_mouthRight*0.216+CTRL_expressions_mouthLipsTowardsUR*0.216+CTRL_expressions_mouthLipsTowardsDR*0.216), 0, 1)\r\nR_lower_lip_up_pose=clamp((CTRL_expressions_mouthLowerLipDepressL*0.352+CTRL_expressions_mouthLowerLipTowardsTeethR*0.648+CTRL_expressions_mouthPressDR*0.5+CTRL_expressions_mouthCornerSharpenDR+CTRL_expressions_jawChinRaiseDL*0.028+CTRL_expressions_jawChinRaiseDR*0.676+CTRL_expressions_mouthRight*0.216), 0, 1)\r\nL_lower_lip_down_pose=clamp((CTRL_expressions_mouthLowerLipDepressL+CTRL_expressions_mouthLowerLipDepressR*0.784+CTRL_expressions_mouthStretchL*0.648+CTRL_expressions_mouthUpperLipBiteL*0.054+CTRL_expressions_mouthUpperLipBiteR*0.054+CTRL_expressions_mouthLowerLipBiteL*0.054+CTRL_expressions_mouthLowerLipBiteR*0.054+CTRL_expressions_mouthFunnelUL*0.125+CTRL_expressions_mouthFunnelUR*0.125+CTRL_expressions_mouthFunnelDL*0.125+CTRL_expressions_mouthFunnelDR*0.125+CTRL_expressions_mouthCornerRounderDL*0.5+CTRL_expressions_mouthRight*0.104), 0, 1)\r\nR_lower_lip_down_pose=clamp((CTRL_expressions_mouthLowerLipDepressL*0.784+CTRL_expressions_mouthLowerLipDepressR+CTRL_expressions_mouthStretchR*0.648+CTRL_expressions_mouthUpperLipBiteL*0.054+CTRL_expressions_mouthUpperLipBiteR*0.054+CTRL_expressions_mouthLowerLipBiteL*0.054+CTRL_expressions_mouthLowerLipBiteR*0.054+CTRL_expressions_mouthFunnelUL*0.125+CTRL_expressions_mouthFunnelUR*0.125+CTRL_expressions_mouthFunnelDL*0.125+CTRL_expressions_mouthFunnelDR*0.125+CTRL_expressions_mouthCornerRounderDR*0.5+CTRL_expressions_mouthLeft*0.104), 0, 1)\r\nL_lower_lip_up_pose=clamp((CTRL_expressions_mouthLowerLipDepressR*0.216+CTRL_expressions_mouthLowerLipTowardsTeethL*0.648+CTRL_expressions_mouthPressDL*0.5+CTRL_expressions_mouthCornerSharpenDL+CTRL_expressions_jawChinRaiseDL*0.676+CTRL_expressions_jawChinRaiseDR*0.028+CTRL_expressions_mouthLeft*0.216), 0, 1)\r\nL_smile_pose=clamp((CTRL_expressions_mouthCornerPullL+CTRL_expressions_mouthDimpleL*0.104+CTRL_expressions_mouthCornerUpL*0.5+CTRL_expressions_mouthLeft*0.5), 0, 1)\r\nR_smile_pose=clamp((CTRL_expressions_mouthCornerPullR+CTRL_expressions_mouthDimpleR*0.104+CTRL_expressions_mouthCornerUpR*0.5+CTRL_expressions_mouthRight*0.5), 0, 1)\r\nL_lip_corner_wide_pose=clamp((CTRL_expressions_mouthStretchL+CTRL_expressions_mouthDimpleL+CTRL_expressions_mouthCornerWideL), 0, 1)\r\nL_frown_pose=clamp((CTRL_expressions_mouthStretchL*0.5+CTRL_expressions_mouthCornerDepressL+CTRL_expressions_mouthCornerDownL*0.784), 0, 1)\r\nR_lip_corner_wide_pose=clamp((CTRL_expressions_mouthStretchR+CTRL_expressions_mouthDimpleR+CTRL_expressions_mouthCornerWideR), 0, 1)\r\nR_frown_pose=clamp((CTRL_expressions_mouthStretchR*0.5+CTRL_expressions_mouthCornerDepressR+CTRL_expressions_mouthCornerDownR*0.784), 0, 1)\r\nL_upper_lip_lower_pose=clamp((CTRL_expressions_mouthUpperLipTowardsTeethL+CTRL_expressions_mouthCornerSharpenUL+CTRL_expressions_jawChinRaiseUR*0.104+CTRL_expressions_mouthRight*0.104), 0, 1)\r\nR_upper_lip_lower_pose=clamp((CTRL_expressions_mouthUpperLipTowardsTeethR+CTRL_expressions_mouthCornerSharpenUR+CTRL_expressions_jawChinRaiseUL*0.104+CTRL_expressions_mouthLeft*0.104), 0, 1)\r\nL_lip_corner_narrow_pose=clamp((CTRL_expressions_mouthCornerNarrowL*0.784+CTRL_expressions_jawOpen*0.352+CTRL_expressions_mouthLipsPurseUL+CTRL_expressions_mouthLipsPurseDL+CTRL_expressions_mouthRight*0.5+CTRL_expressions_mouthLipsTowardsUL*0.784+CTRL_expressions_mouthLipsTowardsDL*0.784), 0, 1)\r\nR_lip_corner_narrow_pose=clamp((CTRL_expressions_mouthCornerNarrowR*0.784+CTRL_expressions_jawOpen*0.352+CTRL_expressions_mouthLipsPurseUR+CTRL_expressions_mouthLipsPurseDR+CTRL_expressions_mouthLeft*0.5+CTRL_expressions_mouthLipsTowardsUR*0.784+CTRL_expressions_mouthLipsTowardsDR*0.784), 0, 1)\r\ntongue_up_pose=CTRL_expressions_tongueUp\r\ntongue_forward_pose=CTRL_expressions_tongueOut\r\ntongue_back_pose=CTRL_expressions_tongueIn\r\njaw_open_pose=CTRL_expressions_jawOpen*0.844\r\njaw_left_pose=clamp((CTRL_expressions_jawLeft+CTRL_expressions_mouthLeft*0.156), 0, 1)\r\njaw_right_pose=clamp((CTRL_expressions_jawRight+CTRL_expressions_mouthRight*0.156), 0, 1)\r\njaw_forward_pose=CTRL_expressions_jawFwd\r\njaw_back_pose=CTRL_expressions_jawBack\r\nphoneme_mbp_pose=CTRL_expressions_mouthUpperLipBiteL*0.25+CTRL_expressions_mouthUpperLipBiteR*0.25+CTRL_expressions_mouthLowerLipBiteL*0.25+CTRL_expressions_mouthLowerLipBiteR*0.25\r\nphoneme_fv_pose=CTRL_expressions_mouthUpperLipBiteL*0.1+CTRL_expressions_mouthUpperLipBiteR*0.1+CTRL_expressions_mouthLowerLipBiteL*0.1+CTRL_expressions_mouthLowerLipBiteR*0.1\r\nphoneme_ch_pose=CTRL_expressions_mouthFunnelUL*0.25+CTRL_expressions_mouthFunnelUR*0.25+CTRL_expressions_mouthFunnelDL*0.25+CTRL_expressions_mouthFunnelDR*0.25\r\nC_glabella_down_pose=clamp((CTRL_expressions_browDINL+CTRL_expressions_browDINR), 0, 1)\r\n\r\n"}
   }
   Remap Curves From Mesh!(SourcePose, CurveExpressions:Map of Names to Strings=@n2, SourceMeshComponent) -> (Pose) {bUseAttachedParent=true, expressionSource=ExpressionMap}
   n3 = "Transform (Modify) Bone - Bone: head"!(ComponentPose=Local To Component(LocalPose=Use cached pose 'preScaleCheck'()), Alpha=1) -> (Pose) {scaleMode=BMM_Replace, alphaBoolBlend {blendOption=Linear}, boneToModify {boneName=head}, scale=<0.01 0.01 0.01>}
   Output Pose(Result=Blend Poses by bool(
     BlendPose_0=Component To Local(ComponentPose=@n3)
     BlendPose_1=Use cached pose 'preScaleCheck'()
     BlendTime_0=0
     BlendTime_1=0
     bActiveValue=Get GetPlayerListUsesCustomCharacterParts() -> (GetPlayerListUsesCustomCharacterParts:bool)
    ) {blendType=Linear}) {layerGroup=None}
   n4 = "Blend Poses (EFortFacialAnimTypes)"!(
    BlendPose_0:Pose Link Structure=Use cached pose 'GameFacial'()
    BlendPose_1:Pose Link Structure=Layered blend per bone(BasePose=Use cached pose 'GameFacial'(), BlendPoses_0=Use cached pose 'LiveLinkFacial'()) {layerSetup [{branchFilters [{boneName=faceAttach}]}]}
    BlendPose_2:Pose Link Structure=Use cached pose 'LiveLinkFacial'()
    BlendTime_0:float=0.1
    BlendTime_1:float=0.1
    BlendTime_2:float=0.1
    ActiveEnumValue:Byte=Get!() -> (
     FacialInput_SubjectName:Live Link Subject Name Structure
     FacialInput_CurrentAnimType:EFortFacialAnimTypes Enum
     FacialInput_bCurvesOnly:bool=~
    ).FacialInput_CurrentAnimType
   ) -> (Pose:Pose Link Structure) {
    is=title AnimGraphNode_BlendListByEnum
    blendType=Linear
    blendPose [{linkId=-1, sourceLinkId=-1}, {linkId=-1, sourceLinkId=-1}, {linkId=-1, sourceLinkId=-1}]
    blendTime [0.1,0.1,0.1]
   }
   n5 = Use cached pose 'IncomingPose'()
   n6 = Get bBypassHeadPartLogic() -> (bBypassHeadPartLogic:bool)
   Save cached pose 'PrePoseAsset'(Pose=Blend Poses by bool(
     BlendPose_0=Blend Poses by bool(BlendPose_0=@n5, BlendPose_1=@n4, BlendTime_0=0, BlendTime_1=0, bActiveValue=@n6)
     BlendPose_1=Use cached pose 'IdlePose'()
     bActiveValue=Get bEnableGameplayFacialCurves() -> (bEnableGameplayFacialCurves:bool)
    )) {cacheName=PrePoseAsset}
   n7 = Two Way Blend!(A=Use cached pose 'IdlePose'(), B=Use cached pose 'IncomingPoseWithBlinks'(), AlphaCurveName:Name=DisableFaceOverride) -> (Pose) {blendNode {alphaInputType=Curve, alphaBoolBlend {blendOption=Linear}}}
   n8 = Use cached pose 'IncomingPose'()
   n9 = Get bypassIdleAnim() -> (bypassIdleAnim:bool)
   n10 = "Blend Poses (EFortFacialAnimTypes)"(
    BlendPose_0=Two Way Blend!(
     A=Use cached pose 'IncomingPoseWithBlinks'()
     B=Blend Poses by bool(BlendPose_0=@n8, BlendPose_1=@n7, bActiveValue=@n9) {blendType=Linear}
     Alpha:float=Get FaceIdleAlpha() -> (FaceIdleAlpha:float)
    ) -> (Pose) {blendNode {alphaBoolBlend {blendOption=Linear}}}
    BlendPose_1=Use cached pose 'IncomingPose'()
    BlendPose_2=Use cached pose 'IncomingPose'()
    BlendTime_0=0.1
    BlendTime_1=0.1
    BlendTime_2=0.1
    ActiveEnumValue=Get() -> (FacialInput_bCurvesOnly=~).FacialInput_CurrentAnimType
   ) {
    blendType=Linear
    blendPose [{linkId=-1, sourceLinkId=-1}, {linkId=-1, sourceLinkId=-1}, {linkId=-1, sourceLinkId=-1}]
    blendTime [0.1,0.1,0.1]
   }
   Save cached pose 'GameFacial'(Pose=@n10) {cacheName=GameFacial}
   n11 = Two Way Blend!(
    A=Blend Poses by bool(
     BlendPose_0=Sequence Player 'Frontend_Default_Face_Idle'() {sequence=/Game/Animation/Game/MainPlayer/Menu/FACIAL/Frontend_Default_Face_Idle}
     BlendPose_1=Use cached pose 'DefaultFaceIdle'()
     bActiveValue=Get bSkipDefaultFaceIdle() -> (bSkipDefaultFaceIdle:bool)
    ) {blendType=Linear}
    B=Slot 'Frontend_CustomFaceIdle'() {slotName=Frontend_CustomFaceIdle}
    AlphaCurveName:Name=Custom_Frontend_Face_Idle
   ) -> (Pose) {blendNode {alphaInputType=Curve, alphaBoolBlend {blendOption=Linear}}}
   n12 = Use cached pose 'DefaultFaceIdle'()
   n13 = Get bIsInFrontEnd()
   Save cached pose 'IdlePose'(Pose=Layered blend per bone(
     BasePose=Use cached pose 'IncomingPose'()
     BlendPoses_0=Blend Poses by bool(BlendPose_0=@n11, BlendPose_1=@n12, bActiveValue=@n13) {blendType=Linear}
    ) {layerSetup [{branchFilters [{boneName=faceAttach}]}]}) {cacheName=IdlePose}
   Save cached pose 'IncomingPoseWithBlinks'(Pose=Blend Facial Curves!(
     FirstPose:Pose Link Structure=Use cached pose 'IncomingPose'()
     SecondPose:Pose Link Structure=Use cached pose 'DefaultFaceIdle'()
     Alpha:float=Get FaceIdleBlinksAlpha() -> (FaceIdleBlinksAlpha:float)
    ) -> (Pose:Pose Link Structure) {is=title FortAnimGraphNode_BlendFacialCurves}) {cacheName=IncomingPoseWithBlinks}
   n14 = Slot 'FaceSlot'(Source=Use cached pose 'PostPoseAsset'()) {slotName=FaceSlot}
   n15 = Use cached pose 'PostPoseAsset'()
   n16 = Get FaceSlotWeight() -> (FaceSlotWeight:float)
   Save cached pose 'preScaleCheck'(Pose=Layered blend per bone(BasePose=Use cached pose 'PostPoseAsset'(), BlendPoses_0=Two Way Blend!(A=@n15, B=@n14, Alpha:float=@n16) -> (Pose)) {layerSetup [{branchFilters [{boneName=faceAttach}]}]}) {cacheName=preScaleCheck}
   n17 = Use cached pose 'GameFacial'()
   n18 = Break Fort Anim Input Facial!(FortAnimInput_Facial:"Fort Anim Input Facial Structure (by ref)"=Get FacialInput() -> (FacialInput:Fort Anim Input Facial Structure)) -> (SubjectName:Live Link Subject Name Structure, CurrentAnimType:EFortFacialAnimTypes Enum, bCurvesOnly:bool=~) {is=title K2Node_BreakStruct}
   n19 = Live Link Pose!(
    InputPose:Pose Link Structure=Use cached pose 'GameFacial'()
    LiveLinkSubjectName:Live Link Subject Name Structure=@n18
    bDoLiveLinkEvaluation:bool=True
   ) -> (Pose:Pose Link Structure) {
    is=title AnimGraphNode_LiveLinkPose
    retargetAsset=/Game/Characters/Player/Common/RemapAssets/FacialAnim_CurvesAndBones_HeadRemapAsset.FacialAnim_CurvesAndBones_HeadRemapAsset_C
   }
   Save cached pose 'LiveLinkFacial'(Pose=Blend Poses by bool(
     BlendPose_0=Live Link Pose(InputPose=@n17, LiveLinkSubjectName=@n18, bDoLiveLinkEvaluation=True) {retargetAsset=/Game/Characters/Player/Common/RemapAssets/FacialAnim_HeadRemapAsset.FacialAnim_HeadRemapAsset_C}
     BlendPose_1=@n19
     bActiveValue=@n18.bCurvesOnly
    ) {blendType=Linear}) {cacheName=LiveLinkFacial}
   n20 = Remap Curves From Mesh!(
    SourcePose=Use cached pose 'PrePoseAsset'()
    CurveExpressionsDataAsset:Curve Expressions Data Asset Object Reference=Get facial_curve_mapping_expression() -> (facial_curve_mapping_expression:Curve Expressions Data Asset Object Reference)
    SourceMeshComponent
   ) -> (Pose) {bUseAttachedParent=true, expressionSource=DataAsset}
   n21 = Use cached pose 'PrePoseAsset'()
   n22 = Get bIs3LAnim() -> (bIs3LAnim:bool)
   n23 = "Blend Poses (EFortniteAnimationType)"!(
    BlendPose_0:Pose Link Structure
    BlendPose_1:Pose Link Structure=Use cached pose 'PrePoseAsset'()
    BlendPose_2:Pose Link Structure=Rig Mapper!(SourcePose:Pose Link Structure=Use cached pose 'PrePoseAsset'(), Alpha:float=1) -> (Pose:Pose Link Structure) {is=title AnimGraphNode_RigMapper, definitions [/RigMapper/Definitions/Baked/RM_CDL_FNL]}
    BlendPose_3:Pose Link Structure=Rig Mapper(SourcePose=Use cached pose 'PrePoseAsset'(), Alpha=1) {definitions [/RigMapper/Definitions/Baked/RM_MHL_FNL]}
    BlendTime_0:float=0.1
    BlendTime_1:float=0.1
    BlendTime_2:float=0.1
    BlendTime_3:float=0.1
    ActiveEnumValue:Byte=Get AnimationType() -> (AnimationType:EFortniteAnimationType Enum=FN_Legacy)
   ) -> (Pose:Pose Link Structure) {
    is=title AnimGraphNode_BlendListByEnum
    blendType=Linear
    blendPose [{linkId=-1, sourceLinkId=-1}, {linkId=-1, sourceLinkId=-1}, {linkId=-1, sourceLinkId=-1}, {linkId=-1, sourceLinkId=-1}]
    blendTime [0.1,0.1,0.1,0.1]
   }
   Save cached pose 'RemapCurvePose'(Pose=Blend Poses by bool(
     BlendPose_0=@n23
     BlendPose_1=Blend Poses by bool(BlendPose_0=@n20, BlendPose_1=@n21, bActiveValue=@n22) {blendType=Linear}
     bActiveValue=Get bUseRigMapper() -> (bUseRigMapper:bool)
    ) {blendType=Linear}) {cacheName=RemapCurvePose}
   Save cached pose 'IncomingPose'(Pose=Copy Pose From Mesh() {bCopyCurves=true, bUseAttachedParent=true}) {cacheName=IncomingPose}
   n24 = Break FN Flipbook Face Region Activation!(FN_Flipbook_Face_Region_Activation:"FN Flipbook Face Region Activation Structure (by ref)"=Get flipbook_face_activation_var() -> (flipbook_face_activation_var:FN Flipbook Face Region Activation Structure)) -> (
    Enable_Brow_1_#1:bool=~
    Enable_Eye_3_#2:bool=~
    Enable_Mouth_5_#3:bool=~
   ) {is=title K2Node_BreakStruct}
   n25 = Control Rig!(
    Source=Use cached pose 'RemapCurvePose'()
    Alpha
    Enable_Brow:bool=@n24
    Enable_Eye:bool=@n24.Enable_Eye_3_#2
    Enable_Mouth:bool=@n24.Enable_Mouth_5_#3
   ) -> (Pose) {
    alphaBoolBlend {blendOption=Linear}
    controlRigAssetReference {blueprintRigClass=/Game/Characters/Player/Common/Fortnite_Base_Head/Facials/ControlRigs/Fortnite_Flipbook2_Face_Mapping_CtrlRig.Fortnite_Flipbook2_Face_Mapping_CtrlRig_C}
   }
   n26 = Use cached pose 'RemapCurvePose'()
   n27 = Get bIsFlipbook_face() -> (bIsFlipbook_face:bool)
   Save cached pose 'PostPoseAsset'(Pose=Fortnite_Base_Head_FacialPose_Export_PoseAsset!(
     SourcePose:Pose Link Structure=Blend Poses by bool(BlendPose_0=@n25, BlendPose_1=@n26, bActiveValue=@n27) {blendType=Linear}
     bIgnorePoseAssetSourceAnimationCurves:bool=Get ignorePoseAssetSourceAnimCurve() -> (ignorePoseAssetSourceAnimCurve:bool)
    ) -> (Pose:Pose Link Structure) {
     is=title AnimGraphNode_PoseBlendNode
     blendOption=Linear
     poseAsset=&/Fortnite_Base_Head_FacialPose_Export_PoseAsset
    }) {cacheName=PostPoseAsset}
   Save cached pose 'DefaultFaceIdle'(Pose=Sequence Player 'DefaultFaceIdle_Legacy'() {sequence=/Game/Animation/Game/MainPlayer/Menu/FACIAL/DefaultFaceIdle_Legacy}) {cacheName=DefaultFaceIdle}
   Sequence Player 'Fortnite_Base_Head_ROM'() {sequence=/Game/Characters/Player/Common/Fortnite_Base_Head/Anim_Tests/Fortnite_Base_Head_ROM}
   Two Way Blend(Alpha=Get FaceIdleAlpha()) {blendNode {alphaBoolBlend {blendOption=Linear}}}
   Use cached pose 'IncomingPose'()
   GetPlaylistUsesCustomCharacterParts!() -> (ReturnValue:bool)
  ]}
 EventGraph {nodes [
   Parent: BlueprintUpdateAnimation!(execute, DeltaTimeX:float=0) -> (then) {is=title K2Node_CallParentFunction}
   > n1 = Sequence!(execute) -> (then_0, then_1, then_2:Exec)
   > Branch(Condition=Get bIsFlipbook_face())
   n2 = GetCurveValue(CurveName=flipbook_face_L_brow_index)
   n3 = "Not Equal (Float)"(A=@n2, B=Get L_Brow_Prev_Index() -> (L_Brow_Prev_Index:float))
   n4 = GetCurveValue(CurveName=flipbook_face_R_brow_index)
   n5 = "Not Equal (Float)"(A=@n4, B=Get R_Brow_Prev_Index() -> (R_Brow_Prev_Index:float))
   n6 = GetCurveValue(CurveName=flipbook_face_L_eye_index)
   n7 = "Not Equal (Float)"(A=@n6, B=Get L_ Eye_Prev_Index() -> (L_ Eye_Prev_Index:float))
   n8 = GetCurveValue(CurveName=flipbook_face_R_eye_index)
   n9 = "Not Equal (Float)"(A=@n8, B=Get R_Eye_Prev_Index() -> (R_Eye_Prev_Index:float))
   n10 = GetCurveValue(CurveName=flipbook_face_mouth_index)
   n11 = "Not Equal (Float)"(A=@n10, B=Get Mouth_Prev_Index() -> (Mouth_Prev_Index:float))
   > SetScalarParameterValueOnMaterials!(
    execute
    self:Mesh Component Object Reference=GetPartSkeletalMeshComponent()
    ParameterName:Name=HasPoseChanged
    ParameterValue:float="To Float (Boolean)"(InBool=OR Boolean!(A=@n7, B=@n9, C:bool=@n11, D:bool=@n3, E:bool=@n5) -> (ReturnValue))
   ) -> (then)
   > Set L_ Eye_Prev_Index(L_ Eye_Prev_Index:float=@n6) -> (Output_Get:float)
   > Set R_Eye_Prev_Index(R_Eye_Prev_Index:float=@n8) -> (Output_Get:float)
   > Set Mouth_Prev_Index(Mouth_Prev_Index:float=@n10) -> (Output_Get:float)
   > Set L_Brow_Prev_Index(L_Brow_Prev_Index:float=@n2) -> (Output_Get:float)
   > Set R_Brow_Prev_Index(R_Brow_Prev_Index:float=@n4) -> (Output_Get:float)
   Branch(execute=@n1.then_1, Condition=Get bHasMaterialDrivenPupils() -> (bHasMaterialDrivenPupils:bool))
   n12 = MakeVector2D(
    X=float - float(A=GetCurveValue(CurveName=look_left_pose), B=GetCurveValue(CurveName=look_right_pose))
    Y=float - float(A=GetCurveValue(CurveName=look_up_pose), B=GetCurveValue(CurveName=look_down_pose))
   )
   n13 = Get Prev_LookDirection() -> (Prev_LookDirection:Vector 2D Structure)
   > SetScalarParameterValueOnMaterials(
    self=GetPartSkeletalMeshComponent()
    ParameterName=HasLookDirectionChanged
    ParameterValue="To Float (Boolean)"(InBool="Not Equal (Vector2D)"!(A:Vector 2D Structure=@n12, B:Vector 2D Structure=@n13, ErrorTolerance:float=0.025) -> (ReturnValue:bool=~) {is=title K2Node_PromotableOperator})
   )
   > Set Prev_LookDirection(Prev_LookDirection:Vector 2D Structure=@n12) -> (Output_Get:Vector 2D Structure)
   Reroute Node(InputPin=@n1.then_2)
   Event BlueprintInitializeAnimation()
   > Set MeshToCopy(MeshToCopy:Skeletal Mesh Component Object Reference=GetOwnerSkeletalMeshComponent()) -> (Output_Get:Skeletal Mesh Component Object Reference)
   > Set GetPlayerListUsesCustomCharacterParts(GetPlayerListUsesCustomCharacterParts:bool=GetPlaylistUsesCustomCharacterParts()) -> (Output_Get:bool)
   n14 = Get Class(Object=GetOwningActor())
   > n15 = Branch(Condition="Equal Exactly (String)"(A=Get Class Display Name(Class=@n14), B=AnimationEditorPreviewActor))
   Set bypassIdleAnim(execute=[@n15.then, @n16.Is Not Valid], bypassIdleAnim:bool=true) -> (Output_Get:bool)
   n16 = Is Valid(exec=@n15.else, InputObject=TryGetPawnOwner())
   > Set bypassIdleAnim(bypassIdleAnim=false)
   Branch(Condition=Get bHasBlinkingEyes() -> (bHasBlinkingEyes:bool))
   n17 = MakeVector2D(X=GetCurveValue(CurveName=L_blink_pose), Y=GetCurveValue(CurveName=R_blink_pose))
   n18 = Get Prev Blink Pose() -> (Prev Blink Pose:Vector 2D Structure)
   n19 = Get Blink Pose Threshold() -> (Blink Pose Threshold:float)
   > SetScalarParameterValueOnMaterials(
    self=GetPartSkeletalMeshComponent()
    ParameterName=HasLidChanged
    ParameterValue="To Float (Boolean)"(InBool="Not Equal (Vector2D)"(A=@n17, B=@n18, ErrorTolerance=@n19) -> (ReturnValue=~))
   )
   > Set Prev Blink Pose(Prev Blink Pose:Vector 2D Structure=@n18) -> (Output_Get:Vector 2D Structure)
   n20 = GetPartSkeletalMeshComponent()
   n21 = Is Valid(InputObject=@n20)
   n22 = GetAnimInstance(self=@n20)
   > n23 = Is Valid(InputObject=@n22)
   Set is_3L_anim_obsolete(execute=[@n24.else, @n23.Is Not Valid, @n21.Is Not Valid], is_3L_anim_obsolete:bool=false) -> (Output_Get:bool)
   n24 = Branch(execute=@n23.Is Valid, Condition="float > float"(A=GetCurveValue(self=@n22, CurveName=DisableFaceOverride), B=0.1))
   > Set is_3L_anim_obsolete(is_3L_anim_obsolete="float > float"(A=GetCurveValue(self=@n22, CurveName=is_3L), B=0.1))
   GetOwnerSkeletalMeshComponent()
  ]}
]
variables [
 bIsCrouching_0 {type=boolean}
 bIsCrouchSprinting_0 {type=boolean}
 bIsCrouchWalking {type=boolean}
 bIsDancing {type=boolean}
 bIsDBNO_0 {type=boolean}
 bIsFallingLocal {type=boolean}
 bIsJogging_0 {type=boolean}
 bIsJumpingLocal {type=boolean}
 bIsLocomotion_0 {type=boolean}
 bIsMeleeSwing_0 {type=boolean}
 bIsMoving {type=boolean}
 bIsOnGround_0 {type=boolean}
 bIsSlopeSliding_0 {type=boolean}
 bIsSprinting_0 {type=boolean}
 bIsTargeting_0 {type=boolean}
 Blink Pose Threshold {type=number, default=0.025}
 bRecentlyFired_0 {type=boolean}
 bypassIdleAnim {type=boolean}
 facial_curve_mapping_expression {type=CurveExpressionsDataAsset, default=/Game/Characters/Player/Common/Fortnite_Base_Head/Facials/CurveMappings/FN_3LToLegacy_Main_Mapping}
 Facial_Curve_Zero_Expression {
  type=CurveExpressionsDataAsset
  default=/Game/Characters/Player/Common/Fortnite_Base_Head/Facials/CurveMappings/FN_ZeroOutLegacyDefaultFaceIdle
 }
 flipbook_face_activation_var {type=FN_Flipbook_Face_Region_Activation, default {enable_Brow=true, enable_Eye=true, enable_Mouth=true}}
 GetPlayerListUsesCustomCharacterParts {type=boolean}
 ignorePoseAssetSourceAnimCurve {type=boolean, default=true}
 is_3L_anim_obsolete {type=boolean}
 IsDive0 {type=boolean}
 IsSkydiving0 {type=boolean}
 L_ Eye_Prev_Index {type=number}
 L_Brow_Prev_Index {type=number}
 LastFireTime {type=number}
 LookPitch {type=number}
 LookYaw {type=number}
 MeshToCopy {type=SkeletalMeshComponent}
 Mouth_Prev_Index {type=number}
 Prev Blink Pose {type=Vector2D, default=<0 0>}
 Prev_LookDirection {type=Vector2D, default=<0 0>}
 R_Brow_Prev_Index {type=number}
 R_Eye_Prev_Index {type=number}
]
